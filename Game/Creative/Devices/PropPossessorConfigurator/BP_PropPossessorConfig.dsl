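; Fortnite Creative Prop Possessor configurator device that registers itself as the prop hunt settings for its minigame.
; On BeginPlay, after the parent call, it looks up its minigame, checks validity and casts it to Minigame_PlayerCreated.
; It then stores a reference to itself in that minigame's PropHuntSettings variable.
; The overlap and tick events are empty, and the construction script only calls the parent.
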(event EventBeginPlay
  (|CallParentFunction)
  (CallFunction|SetMinigamePropHuntSettings))

(event Collision|EventActorBeginOverlap
  (params (OtherActor :type Actor%20Object%20Reference)))

(event EventTick
  (params (DeltaSeconds :type Float%20%28single-precision%29)))

(fn SetMinigamePropHuntSettings
  (bind _self self)
  (bind _returnvalue (Minigame|GetMinigame _self :!class /Script/FortniteGame.MinigameBlueprintLibrary))
  (if (Utilities|IsValid _returnvalue :!class /Script/Engine.KismetSystemLibrary)
    (bind _asminigame_player_created (Utilities|Casting|CastToMinigame_PlayerCreated _returnvalue))
    (Variables|/Game/Athena/Playset/Minigames/Minigame_PlayerCreated.Minigame_PlayerCreated_C|SetPropHuntSettings self _asminigame_player_created)))

(fn ConstructionScript
  (|CallParentFunction))
